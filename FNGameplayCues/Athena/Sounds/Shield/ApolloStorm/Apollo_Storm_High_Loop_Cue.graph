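# Apollo_Storm_High_Loop_Cue is a looping sound cue for the high layer of the Apollo storm audio, stored among the Athena gameplay-cue shield sounds.
# A single looping wave player feeds the cue output directly from the Apollo_Storm_High_Loop_st wave.
# The cue plays in the Fort storms sound class with raised volume, a very large max distance and a pre-effect bus send entry.
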SoundCue /FNGameplayCues/Athena/Sounds/Shield/ApolloStorm/Apollo_Storm_High_Loop_Cue
root=WavePlayer_5
nodes [{id=WavePlayer_5, properties {bLooping=true, soundWaveAssetPtr=&/Apollo_Storm_High_Loop_st}}]
properties {maxDistance=2097152, soundClassObject=/Game/Sounds/FortSoundClassesAndModes/Fort_Storms_PSM, volumeMultiplier=1.416, preEffectBusSends [{}]}
