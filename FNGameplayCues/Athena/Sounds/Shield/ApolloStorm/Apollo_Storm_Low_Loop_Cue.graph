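# Sound cue for the low looping Apollo storm ambience, routed through the Fort_Storms_PSM sound class.
# A Mixer layers a medium storm loop with a heavy wind loop, weighting the wind louder and passing it through a Modulator.
# A Quality Level node picks the storm loop's base wave for the two lower tiers and its "_st" variant (likely stereo) for the two higher ones.
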SoundCue /FNGameplayCues/Athena/Sounds/Shield/ApolloStorm/Apollo_Storm_Low_Loop_Cue
root=Mixer_0
nodes [
 {id=Mixer_0, children [QualityLevel_1, Modulator_0], properties {inputVolume [1.2,2.5]}}
 {id=Modulator_0, children [WavePlayer_0]}
 {id=QualityLevel_1, children [WavePlayer_6,WavePlayer_6,WavePlayer_7,WavePlayer_7]}
 {id=WavePlayer_0, properties {bLooping=true, soundWaveAssetPtr=&/Apollo_Storm_WindHeavy_Loop}}
 {id=WavePlayer_6, properties {bLooping=true, soundWaveAssetPtr=&/Apollo_Storm_Medium_Loop}}
 {id=WavePlayer_7, properties {bLooping=true, soundWaveAssetPtr=&/Apollo_Storm_Medium_Loop_st}}
]
properties {maxDistance=2097152, soundClassObject=/Game/Sounds/FortSoundClassesAndModes/Fort_Storms_PSM, volumeMultiplier=1.416, preEffectBusSends [{}]}
